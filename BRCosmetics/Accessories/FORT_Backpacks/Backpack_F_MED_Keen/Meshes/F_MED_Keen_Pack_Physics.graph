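# Physics asset for the Keen backpack cosmetic, defining physical bodies for the pack and its seven arrow bones.
# Each bone carries one small capsule of identical size.
# Every arrow is constrained directly to the pack rather than chained to the others.
# Narrow swing limits that vary per arrow keep each one close to its rest angle, while all share a wider twist limit.
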PhysicsAsset /BRCosmetics/Accessories/FORT_Backpacks/Backpack_F_MED_Keen/Meshes/F_MED_Keen_Pack_Physics
bodies {
 arrow01 [{shapeType=Capsule, length=2, radius=2, shapeName=arrow01_capsule}]
 arrow02 [{shapeType=Capsule, length=2, radius=2, shapeName=arrow02_capsule}]
 arrow03 [{shapeType=Capsule, length=2, radius=2, shapeName=arrow03_capsule}]
 arrow04 [{shapeType=Capsule, length=2, radius=2, shapeName=arrow04_capsule}]
 arrow05 [{shapeType=Capsule, length=2, radius=2, shapeName=arrow05_capsule}]
 arrow06 [{shapeType=Capsule, length=2, radius=2, shapeName=arrow06_capsule}]
 arrow07 [{shapeType=Capsule, length=2, radius=2, shapeName=arrow07_capsule}]
 pack [{shapeType=Capsule, length=2, radius=2, shapeName=pack_capsule}]
}
constraints [
 {bone1Name=arrow02, bone2Name=pack, swing1LimitDegrees=12, swing2LimitDegrees=10, twistLimitDegrees=70}
 {bone1Name=arrow03, bone2Name=pack, swing1LimitDegrees=10, swing2LimitDegrees=10, twistLimitDegrees=70}
 {bone1Name=arrow04, bone2Name=pack, swing1LimitDegrees=15, swing2LimitDegrees=10, twistLimitDegrees=70}
 {bone1Name=arrow05, bone2Name=pack, swing1LimitDegrees=10, swing2LimitDegrees=15, twistLimitDegrees=70}
 {bone1Name=arrow06, bone2Name=pack, swing1LimitDegrees=20, swing2LimitDegrees=20, twistLimitDegrees=70}
 {bone1Name=arrow07, bone2Name=pack, swing1LimitDegrees=10, swing2LimitDegrees=10, twistLimitDegrees=70}
 {bone1Name=arrow01, bone2Name=pack, swing1LimitDegrees=6, swing2LimitDegrees=7, twistLimitDegrees=70}
]
