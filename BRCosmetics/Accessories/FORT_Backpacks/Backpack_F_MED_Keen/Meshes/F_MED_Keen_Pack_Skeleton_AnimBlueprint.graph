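# Animation Blueprint for the F_MED_Keen backpack, adding secondary physics motion to its seven arrow bones.
# A RigidBody node in pack-bone space feeds chained per-arrow Anim Dynamics nodes with angular springs and tight limits, then a Modify Bone offsets, rotates and shrinks the pack.
# Each update sets dynAlpha and rigidAlpha lower in the front end, though the graph shows no nodes reading them.
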AnimBlueprint /BRCosmetics/Accessories/FORT_Backpacks/Backpack_F_MED_Keen/Meshes/F_MED_Keen_Pack_Skeleton_AnimBlueprint
target_skeleton=&/F_MED_Keen_Pack_Skeleton
graphs [
 AnimGraph {nodes [
   n1 = RigidBody(ComponentPose=Local To Component(LocalPose=Local Space Ref Pose())) {
    alpha=0
    overridePhysicsAsset=&/F_MED_Keen_Pack_Physics
    simulationSpace=BaseBoneSpace
    alphaBoolBlend {blendOption=Linear}
    baseBoneRef {boneName=pack}
   }
   n2 = Anim Dynamics - Bone: arrow01(ComponentPose=@n1, Alpha=1, ExternalForce="0,0,0") {
    alpha=0
    angularDampingOverride=1
    angularSpringConstant=1500
    bAngularSpring=true
    bGravityOverrideInSimSpace=true
    bOverrideAngularDamping=true
    bUseGravityOverride=true
    gravityScale=0.35
    simulationSpace=RootRelative
    alphaBoolBlend {blendOption=Linear}
    boundBone {boneName=arrow01}
    componentLinearAccScale=<0.1 0.1 0.1>
    componentLinearVelScale=<0.1 0.1 0.1>
    gravityOverride {z=-980}
    physicsBodyDefinitions [{
      sphereCollisionRadius=0
      boundBone {boneName=arrow01}
      boxExtents=<15 10 10>
      constraintSetup {angularLimitsMax=<3 5 5>, angularLimitsMin=<-3 -5 -1>, angularTarget=<0 0 0>}
      localJointOffset=<5 0 0>
     }]
    relativeSpaceBone {boneName=arrow01}
   }
   n3 = Anim Dynamics - Bone: arrow02(ComponentPose=@n2, Alpha=1, ExternalForce="0,0,0") {
    alpha=0
    angularDampingOverride=1
    angularSpringConstant=1500
    bAngularSpring=true
    bGravityOverrideInSimSpace=true
    bOverrideAngularDamping=true
    bUseGravityOverride=true
    gravityScale=0.35
    simulationSpace=RootRelative
    alphaBoolBlend {blendOption=Linear}
    boundBone {boneName=arrow02}
    componentLinearAccScale=<0.1 0.1 0.1>
    componentLinearVelScale=<0.1 0.1 0.1>
    gravityOverride {z=-980}
    physicsBodyDefinitions [{
      sphereCollisionRadius=0
      boundBone {boneName=arrow02}
      boxExtents=<15 10 10>
      constraintSetup {angularLimitsMax=<3 5 10>, angularLimitsMin=<-3 -5 0>, angularTarget=<0 0 0>}
      localJointOffset=<5 0 0>
     }]
    relativeSpaceBone {boneName=arrow02}
   }
   n4 = Anim Dynamics - Bone: arrow03(ComponentPose=@n3, Alpha=1, ExternalForce="0,0,0") {
    alpha=0
    angularDampingOverride=1
    angularSpringConstant=1500
    bAngularSpring=true
    bGravityOverrideInSimSpace=true
    bOverrideAngularDamping=true
    bUseGravityOverride=true
    gravityScale=0.35
    simulationSpace=RootRelative
    alphaBoolBlend {blendOption=Linear}
    boundBone {boneName=arrow03}
    componentLinearAccScale=<0.1 0.1 0.1>
    componentLinearVelScale=<0.1 0.1 0.1>
    gravityOverride {z=-980}
    physicsBodyDefinitions [{
      sphereCollisionRadius=0
      boundBone {boneName=arrow03}
      boxExtents=<15 10 10>
      constraintSetup {angularLimitsMax=<3 8 5>, angularLimitsMin=<-3 -6 -5>, angularTarget=<0 0 0>}
      localJointOffset=<5 0 0>
     }]
    relativeSpaceBone {boneName=arrow03}
   }
   n5 = Anim Dynamics - Bone: arrow04(ComponentPose=@n4, Alpha=1, ExternalForce="0,0,0") {
    alpha=0
    angularDampingOverride=1
    angularSpringConstant=1500
    bAngularSpring=true
    bGravityOverrideInSimSpace=true
    bOverrideAngularDamping=true
    bUseGravityOverride=true
    gravityScale=0.35
    simulationSpace=RootRelative
    alphaBoolBlend {blendOption=Linear}
    boundBone {boneName=arrow04}
    componentLinearAccScale=<0.1 0.1 0.1>
    componentLinearVelScale=<0.1 0.1 0.1>
    gravityOverride {z=-980}
    physicsBodyDefinitions [{
      sphereCollisionRadius=0
      boundBone {boneName=arrow04}
      boxExtents=<15 10 10>
      constraintSetup {angularLimitsMax=<3 5 5>, angularLimitsMin=<-3 -5 -1>, angularTarget=<0 0 0>}
      localJointOffset=<5 0 0>
     }]
    relativeSpaceBone {boneName=arrow04}
   }
   n6 = Anim Dynamics - Bone: arrow05(ComponentPose=@n5, Alpha=1, ExternalForce="0,0,0") {
    alpha=0
    angularDampingOverride=1
    angularSpringConstant=1500
    bAngularSpring=true
    bGravityOverrideInSimSpace=true
    bOverrideAngularDamping=true
    bUseGravityOverride=true
    gravityScale=0.35
    simulationSpace=RootRelative
    alphaBoolBlend {blendOption=Linear}
    boundBone {boneName=arrow05}
    componentLinearAccScale=<0.1 0.1 0.1>
    componentLinearVelScale=<0.1 0.1 0.1>
    gravityOverride {z=-980}
    physicsBodyDefinitions [{
      sphereCollisionRadius=0
      boundBone {boneName=arrow05}
      boxExtents=<15 10 10>
      constraintSetup {angularLimitsMax=<3 2 5>, angularLimitsMin=<-3 -10 -2>, angularTarget=<0 0 0>}
      localJointOffset=<5 0 0>
     }]
    relativeSpaceBone {boneName=arrow05}
   }
   n7 = Anim Dynamics - Bone: arrow06(ComponentPose=@n6, Alpha=1, ExternalForce="0,0,0") {
    alpha=0
    angularDampingOverride=1
    angularSpringConstant=1500
    bAngularSpring=true
    bGravityOverrideInSimSpace=true
    bOverrideAngularDamping=true
    bUseGravityOverride=true
    gravityScale=0.35
    simulationSpace=RootRelative
    alphaBoolBlend {blendOption=Linear}
    boundBone {boneName=arrow06}
    componentLinearAccScale=<0.1 0.1 0.1>
    componentLinearVelScale=<0.1 0.1 0.1>
    gravityOverride {z=-980}
    physicsBodyDefinitions [{
      sphereCollisionRadius=0
      boundBone {boneName=arrow06}
      boxExtents=<15 10 10>
      constraintSetup {angularLimitsMax=<3 5 10>, angularLimitsMin=<-3 -5 -10>, angularTarget=<0 0 0>}
      localJointOffset=<5 0 0>
     }]
    relativeSpaceBone {boneName=arrow06}
   }
   n8 = Anim Dynamics - Bone: arrow07(ComponentPose=@n7, Alpha=1, ExternalForce="0,0,0") {
    alpha=0
    angularDampingOverride=1
    angularSpringConstant=1500
    bAngularSpring=true
    bGravityOverrideInSimSpace=true
    bOverrideAngularDamping=true
    bUseGravityOverride=true
    gravityScale=0.35
    simulationSpace=RootRelative
    alphaBoolBlend {blendOption=Linear}
    boundBone {boneName=arrow07}
    componentLinearAccScale=<0.1 0.1 0.1>
    componentLinearVelScale=<0.1 0.1 0.1>
    gravityOverride {z=-980}
    physicsBodyDefinitions [{
      sphereCollisionRadius=0
      boundBone {boneName=arrow07}
      boxExtents=<15 10 10>
      constraintSetup {angularLimitsMax=<3 5 10>, angularLimitsMin=<-3 -5 -2>, angularTarget=<0 0 0>}
      localJointOffset=<5 0 0>
     }]
    relativeSpaceBone {boneName=arrow07}
   }
   n9 = "Transform (Modify) Bone - Bone: pack"(ComponentPose=@n8, Translation="-1,-1.5,0", Rotation="0,-5,0", Scale="0.8,0.8,0.8") {
    rotationMode=BMM_Additive
    scaleMode=BMM_Replace
    translationMode=BMM_Replace
    alphaBoolBlend {blendOption=Linear}
    boneToModify {boneName=pack}
    rotation {yaw=-5}
    scale=<0.8 0.8 0.8>
    translation=<-1 -1.5>
   }
   Output Pose(Result=Component To Local(ComponentPose=@n9)) {layerGroup=None}
  ]}
 EventGraph {nodes [
   Event BlueprintUpdateAnimation()
   > n1 = Branch(Condition=Get bIsInFrontEnd())
   > Set dynAlpha(dynAlpha:double=0.25) -> (Output_Get:double)
   > Set rigidAlpha(rigidAlpha:double=0) -> (Output_Get:double)
   Set dynAlpha(execute=@n1.else, dynAlpha=1)
   > Set rigidAlpha(rigidAlpha=1)
   TryGetPawnOwner()
  ]}
]
variables [
 dynAlpha {type=number}
 rigidAlpha {type=number}
]
